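; Passive gameplay ability that acts as a hidden granter for the Season 33 shotgun weapon expertise restock quest.
; On activation it looks up the restock_01 and restock_02 quests on the owning actor.
; If neither exists, it grants the quest from its Quest Defenition variable through the QuestInstanceManager subsystem, then ends.
; EndAbility is called only on that grant path, so an existing quest leaves the ability without an explicit end.
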(event Ability|EventFailedToActivatePassiveAbility
  (params (ActorInfo :type Gameplay%20Ability%20Actor%20Info%20Structure))
  (|CallParentFunction ActorInfo))

(event Ability|EventActivateAbility
  (bind _returnvalue (Ability|GetOwningActorfromActorInfo))
  (bind _returnvalue_2 (Ability|GetOwningActorfromActorInfo))
  ; Check if [quest_s33_weaponexpertise_shotgun_restock_01] is Active... if Not Active...
  (|CallParentFunction)
  (bind _returnvalue_3 (Quest|GetQuestfromDefinition _returnvalue "/BattlePassS33_Quests/Items/QuestItems/WeaponExpertise/Shotgun/quest_s33_weaponexpertise_shotgun_restock_01.quest_s33_weaponexpertise_shotgun_restock_01"))
  (Utilities|IsValid _returnvalue_3
    (:Is%20Not%20Valid
      ; Check if [quest_s33_weaponexpertise_shotgun_restock_02] is Active... if Not Active...
      (bind _returnvalue_4 (Quest|GetQuestfromDefinition _returnvalue "/BattlePassS33_Quests/Items/QuestItems/WeaponExpertise/Shotgun/quest_s33_weaponexpertise_shotgun_restock_02.quest_s33_weaponexpertise_shotgun_restock_02"))
      (Utilities|IsValid _returnvalue_4
        (:Is%20Not%20Valid
          ; Grant [quest_s33_weaponexpertise_shotgun_restock_01] to the player
          (Utilities|IsValid _returnvalue_2
            (:Is%20Valid
              (Class|QuestInstanceManagerSubsystem|GrantQuest (WorldSubsystems|GetQuestInstanceManagerSubsystem) (Utilities|Struct|MakeGrantQuestBlueprintParameters (Variables|Default|GetQuest%20Defenition) _returnvalue_2 true))
              (Ability|EndAbility))))))))

(event Ability|EventOnEndAbility
  (params (bWasCancelled :type Boolean)))
